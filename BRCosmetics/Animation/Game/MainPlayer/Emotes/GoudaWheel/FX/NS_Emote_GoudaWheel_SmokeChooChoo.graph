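# Niagara system for train-style smoke puffs in the GoudaWheel emote effects, built around a single Smoke emitter.
# System scripts run SystemState, a variable-assignment step and a time-of-day color dynamic input.
# User parameters expose front-end and preview flags, cloud delay, fade distance and smoke alpha.
# No renderers or data interfaces are listed, so the emitter's visual setup is not shown.
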NiagaraSystem /BRCosmetics/Animation/Game/MainPlayer/Emotes/GoudaWheel/FX/NS_Emote_GoudaWheel_SmokeChooChoo
dependencies {
 usedDataInterfaces []
 usedDynamicInputs [&FNTimeOfDayColor]
 usedModules [
  &self:SystemSpawnScript.SystemScriptSource.SystemScriptGraph.NiagaraNodeAssignment_1.SetVariables_#1
  &SystemState
 ]
}
emitters {Smoke {rendererClasses [], renderers []}}
summary {
 systemName=NS_Emote_GoudaWheel_SmokeChooChoo
 emitters [{emitterName=Smoke, rendererClasses []}]
 userVariables [
  User.bIsFrontEnd {type=NiagaraBool}
  User.bIsFrontEndPreview {type=NiagaraBool}
  User.cloud_Dleay {type=NiagaraFloat, defaultValue=0.15, description=""}
  User.FadeMaxDistance {type=NiagaraFloat, defaultValue=0, description=""}
  User.WheelalphaSmoke2 {type=NiagaraFloat, defaultValue=0.25, description=""}
 ]
}
