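# Niagara ribbon smoke-trail effect for the GoudaWheel player emote.
# One emitter spawns short-lived particles at sockets across 23 partitions, with sampled socket velocity setting ribbon UV distance.
# Curves shape width and animate material erosion and fade, while the system's time-of-day color alpha scales opacity.
# Hardcoded arrays tint each partition yellow, green, white, red or blue and vary width by SpawnGroup, which is separate from the partition index.
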NiagaraSystem /BRCosmetics/Animation/Game/MainPlayer/Emotes/GoudaWheel/FX/NS_Emote_GoudaWheel_SmokeRibbion
dependencies {
 usedDataInterfaces [
  &NiagaraDataInterfaceCurve
  &NiagaraDataInterfaceColorCurve
  &NiagaraDataInterfaceArrayFloat
  &NiagaraDataInterfaceArrayFloat3
  &NiagaraDataInterfaceSocketReader
 ]
 usedDynamicInputs [&FNTimeOfDayColor]
 usedModules [
  &self:SystemSpawnScript.SystemScriptSource.SystemScriptGraph.NiagaraNodeAssignment_3.SetVariables_#1
  &SystemState
  &self:SystemSpawnScript.SystemScriptSource.SystemScriptGraph.NiagaraNodeAssignment_4.SetVariables_#2
 ]
}
emitters {Ribbon {
  emitterUpdateScript [
   &EmitterState()
   &SpawnRate(SpawnRate=300, Use Spawn Probability=false)
  ]
  particleSpawnScript [
   &PartitionParticles(Number Of Partitions=23)
   &InitializeParticle(
    Color Mode=::NewEnumerator0
    Lifetime=0.5
    Ribbon Width=&RandomRangeFloat(Minimum=20, Maximum=30)
    Ribbon Width Mode=::NewEnumerator1
    UsePositionOffset=false
    Write Material Random=true
   )
   &SocketLocation(
    SamplePosOffset=false
    Socket Index=$Particles.PartitionParticles.Partition
    Socket Reader=$System.Socket Reader
    Socket Sampling Mode=::NewEnumerator3
    VelocitySpeedLimit=false
   )
   /SetVariables(Particles.Position=&AddVectorToPosition(Position=$Output.SocketLocation.SampledPosition, Vector=<0 0 0>)) {isSetParameters=true}
   /SetVariables(Particles.UVDistance:NiagaraFloat=&Multiply_Float(A=&VectorLength(Vector=$Output.SocketLocation.SampledVelocity), B=0.5)) {isSetParameters=true}
  ]
  particleUpdateScript [
   &ParticleState()
   /SetVariables(Particles.RibbonUVDistance:NiagaraFloat=$Particles.UVDistance) {isSetParameters=true}
   &ScaleRibbonWidth(Ribbon Width Scale=&FloatFromCurve(
     FloatCurve=DataInterface() {propertyValues {Curve {keys [
         {arriveTangent=2.331, interpMode=RCIM_Cubic, leaveTangent=2.331, tangentMode=RCTM_User, time=0, value=0.2}
         {arriveTangent=0.2911, interpMode=RCIM_Cubic, leaveTangent=0.2911, tangentMode=RCTM_User, time=0.0438, value=0.275}
         {arriveTangent=1.32, interpMode=RCIM_Cubic, leaveTangent=1.32, tangentMode=RCTM_User, time=0.5, value=1.9578}
         {arriveTangent=0.3935, interpMode=RCIM_Cubic, leaveTangent=0.3935, tangentMode=RCTM_User, time=1, value=0}
        ]}}}
     Scale Curve=&RandomRangeFloat(Minimum=1, Maximum=1)
    )) {module=ScaleRibbonWidth001}
   &FN_RibbonMaterial(
    Color=&ColorFromCurve(ColorCurve=DataInterface() {propertyValues {
       AlphaCurve {keys [<at 0 2 RCIM_Cubic>,<at 0.9843 0 RCIM_Cubic>]}
       BlueCurve {keys [<at 0.0258 0.5089>]}
       GreenCurve {keys [<at 0.0258 0.4793>]}
       RedCurve {keys [<at 0.0258 0.4735>]}
      }})
    Erosion Amount:NiagaraFloat=&Lerp_Float(A=0.25, B=1, Alpha=$Particles.NormalizedAge)
    Erossion Scale U=4.5
    Macro Distortion Amount=0.1
    Macro Pan Offset=-0.1216
    Micro Distortion Amount=0.0384
    Opacity Erosion:NiagaraFloat=&FloatFromCurve(FloatCurve=DataInterface() {propertyValues {Curve {keys [<at 0 0.25 RCIM_Cubic>, {arriveTangent=1.1077, tangentMode=RCTM_User, time=1, value=1}]}}})
    Opacity Power:NiagaraFloat=&FloatFromCurve(FloatCurve=DataInterface() {propertyValues {Curve {keys [<at 0 0.55 RCIM_Cubic>, {arriveTangent=1.1077, tangentMode=RCTM_User, time=1, value=1}]}}})
    Scale Alpha:NiagaraFloat=&FloatFromCurve(
     FloatCurve=DataInterface() {propertyValues {Curve {keys [<at 0 0.7 RCIM_Cubic>, {arriveTangent=-0.4965, tangentMode=RCTM_User, time=1, value=0}]}}}
     Scale Curve=&FloatFromCurve(
      FloatCurve=DataInterface() {propertyValues {Curve {
         defaultValue=0
         keys [
          {arriveTangent=9.0725, interpMode=RCIM_Cubic, leaveTangent=9.0725, tangentMode=RCTM_User, time=0, value=0}
          {arriveTangent=-2.5249, interpMode=RCIM_Cubic, leaveTangent=-2.5249, tangentMode=RCTM_User, time=0.1408, value=0.9847}
          {arriveTangent=-0.0405, interpMode=RCIM_Cubic, leaveTangent=-0.0405, tangentMode=RCTM_User, time=1, value=0}
         ]
        }}}
      Scale Curve=0.2
     )
    )
    UV Basis:NiagaraFloat=&ReturnNormalizedExecIndex()
   ) {module=FN_RibbonMaterial001}
   &ScaleColor(
    Color Value To Scale=$Particles.Color
    Scale Alpha=&MakeFloatFromLinearColor(LinearColor=$System.TOD Color, Channel:ENiagaraFloatFromLinearColorOptions=::NewEnumerator3)
    ScaleA=true
    ScaleRGB=false
   ) {module=ScaleColor002}
   &SolveForcesAndVelocity(Clamp Velocity=false, Limit Acceleration=false)
   &ScaleRibbonWidth(Ribbon Width Scale=&MakeFloatFromVector2d(Vector 2D=&Vector2DFromFloat(Value=&SelectFloatFromArray(
       Direct Array Index=&SpawnGroup()
       Float Selection Array=DataInterface() {propertyValues {FloatData [0.5,0.8,0.5,0.3,0.3,0.3,0.3,0.3,0.3,0.8,0.8,0.8,0.8,0.5,0.5,0.3,0.3,0.3,0.3,0.4,0.3,0.4,0.4]}}
      ))))
   &ScaleColor(
    Scale Alpha=0.15
    Scale RGB=&SelectVectorFromArray(
     Direct Array Index=$Particles.PartitionParticles.Partition
     Vector Selection Array=DataInterface() {propertyValues {FloatData [
        <1 0.85 0.032>
        <1 0.85 0.032>
        <0 1 0>
        <1 1 1>
        <1 1 1>
        <1 0 0>
        <1 0 0>
        <1 0 0>
        <1 0 0>
        <1 0 0>
        <1 0 0>
        <0 1 0>
        <0 1 0>
        <0 1 0>
        <0 1 0>
        <0 1 0>
        <0 1 0>
        <0 1 0>
        <0 1 0>
        <0 0 1>
        <0 0 1>
        <0 0 1>
        <0 0 1>
       ]}}
    )
    ScaleA=true
    ScaleRGB=true
   )
  ]
  rendererClasses [&NiagaraRibbonRendererProperties]
  renderers [{rendererClass=&NiagaraRibbonRendererProperties}]
 }}
summary {
 systemName=NS_Emote_GoudaWheel_SmokeRibbion
 emitters [{emitterName=Ribbon, rendererClasses [&NiagaraRibbonRendererProperties]}]
 userVariables [
  User.bisFootHidden {type=NiagaraBool}
  User.bIsFrontEnd {type=NiagaraBool}
  User.bIsFrontEndPreview {type=NiagaraBool}
  User.Debug {type=NiagaraBool}
  User.DebugVelocity {type=Vector3f, description="", defaultValue {struct=Vector3f, value=<0 0 1>}}
  User.EmissiveColor {type=LinearColor, description="", defaultValue {struct=LinearColor, value=<rgba 0.0196 0.2745 0.5255 1>}}
  User.ParaGlideLeanAlpha {type=NiagaraFloat, defaultValue=0, description=""}
 ]
}
