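# Niagara system for the GoudaWheel emote in the main player's cosmetics, drawing a plane's smoke trail and thruster flames.
# NE_Smoke spawns short-lived sprites thrown backward with drag, fading from red to dark and tinted by time-of-day color; its velocity-gated spawn rate has identical branches.
# NE_Thrusters bursts two spinning particles drawn as meshes and sprite, flickering through waveform-driven scale and color.
# Light and Glow emitters have no renderers.
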NiagaraSystem /BRCosmetics/Animation/Game/MainPlayer/Emotes/GoudaWheel/FX/NS_Emote_GoudaWheel_SmokeTrailPlane
dependencies {
 usedDataInterfaces [&NiagaraDataInterfaceColorCurve, &NiagaraDataInterfaceCurve, &NiagaraDataInterfaceSocketReader]
 usedDynamicInputs [
  &SimulationPosition
  &MakeLinearColorFromVectorAndFloat
  &FNTimeOfDayColor
 ]
 usedModules [
  &SystemState
  &self:SystemSpawnScript.SystemScriptSource.SystemScriptGraph.NiagaraNodeAssignment_0.SetVariables_#1
  &self:SystemSpawnScript.SystemScriptSource.SystemScriptGraph.NiagaraNodeAssignment_2.SetVariables_#2
  &self:SystemSpawnScript.SystemScriptSource.SystemScriptGraph.NiagaraNodeAssignment_1.SetVariables_#3
  &BlendSpaceHelper
 ]
}
emitters {
 Glow {rendererClasses [], renderers []}
 Light {rendererClasses [], renderers []}
 NE_Smoke {
  emitterSpawnScript [/SetVariables(Emitter.FE Particle Spawn Rate=65, Emitter.Max Traversal Speed=600, Emitter.Particle Spawn Rate=65) {isSetParameters=true}]
  emitterUpdateScript [
   &EmitterState(Life Cycle Mode=::NewEnumerator1, Loop Duration=6.2, UseLoopDelay=false)
   &SpawnRate(
    SpawnRate=&FN_FrontEndFloatModifier(
     Gameplay=&MakeCustomFloatFromBool(
      Bool=&SetBoolByFloatComparison(A=&VectorLength(Vector=$User.OwnerLocalVelocity), B=150)
      True Float=$Emitter.Particle Spawn Rate
      False Float=$Emitter.Particle Spawn Rate
     )
     Front End=$Emitter.FE Particle Spawn Rate
    )
    Use Spawn Probability=false
   )
  ]
  particleSpawnScript [
   &InitializeParticle(
    Color Mode=::NewEnumerator0
    Lifetime=&RandomRangeFloat(Minimum=0.3, Maximum=0.5)
    Mesh Scale Mode=::NewEnumerator1
    Mesh Uniform Scale Max=0.6
    Mesh Uniform Scale Min=0.5
    Sprite Size Mode=::NewEnumerator1
    Uniform Sprite Size Max=70
    Uniform Sprite Size Min=60
    UsePositionOffset=true
    Write Material Random=true
   )
   &ShapeLocation(
    MaskSpawnGroup=false
    Non Uniform Scale=<4 1 2>
    Offset Coordinate Space=::Local
    Offset Mode=::NewEnumerator0
    Override Seed=false
    Sphere Radius=4
    UseOwnerScale=true
   )
   &AddVelocity(Velocity=<2 0 0.2>, Velocity Speed Scale=&RandomRangeFloat(Minimum=-300, Maximum=-500)) {module=AddVelocity001}
  ]
  particleUpdateScript [
   &ParticleState()
   &ScaleColor(
    Linear Color Curve=DataInterface() {propertyValues {
      AlphaCurve {keys [
        {arriveTangent=0.4512, interpMode=RCIM_Cubic, leaveTangent=0.4512, tangentMode=RCTM_User, time=0.002, value=0}
        {arriveTangent=-1.3904, interpMode=RCIM_Cubic, leaveTangent=-1.3904, tangentMode=RCTM_User, time=0.4196, value=0.2}
        {arriveTangent=-0.5907, interpMode=RCIM_Cubic, leaveTangent=-0.5907, tangentMode=RCTM_User, time=0.7029, value=0.1}
        <at 1 0 RCIM_Cubic>
       ]}
      BlueCurve {keys [<at 0.1371 0 RCIM_Cubic>,<at 0.2197 0 RCIM_Cubic>]}
      GreenCurve {keys [<at 0.1371 0.0568 RCIM_Cubic>,<at 0.2197 0 RCIM_Cubic>]}
      RedCurve {keys [<at 0.1371 1 RCIM_Cubic>,<at 0.2197 0 RCIM_Cubic>]}
     }}
    Scale Mode=::NewEnumerator2
   )
   &ScaleSpriteSize(Uniform Curve Sprite Scale=DataInterface() {propertyValues {Curve {keys [<at 0 0.65 RCIM_Cubic>, {arriveTangent=0.35, interpMode=RCIM_Cubic, leaveTangent=0.35, time=0.5, value=1}, <at 1 1 RCIM_Cubic>]}}})
   &ScaleMeshSize(Scale Factor=&VectorFromFloat(Value=&FloatFromCurve(FloatCurve=DataInterface() {propertyValues {Curve {keys [
          {arriveTangent=7.3539, time=0, value=0}
          {arriveTangent=-0.0014, interpMode=RCIM_Cubic, leaveTangent=-0.0014, tangentMode=RCTM_User, time=0.1, value=1}
          {arriveTangent=-1.0747, interpMode=RCIM_Cubic, leaveTangent=-1.0747, tangentMode=RCTM_User, time=1, value=0}
         ]}}}))) {module=ScaleMeshSize001}
   &Drag(Drag=5, Use Linear Drag=true, Use Rotational Drag=false)
   &InheritVelocity(
    Inherited Velocity Amount Scale=<-1 -1 -1>
    Inherited Velocity Speed Limit=5
    Source Speed Threshold=0.05
    Speed Limit=true
    UseSpeedThreshold=true
   )
   &SolveForcesAndVelocity(Clamp Velocity=false, Limit Acceleration=false)
   &ScaleColor(
    Color Value To Scale=$Particles.Color
    Scale Alpha=&MakeFloatFromLinearColor(LinearColor=$System.TOD Color, Channel:ENiagaraFloatFromLinearColorOptions=::NewEnumerator3)
    Scale RGB=&MakeVectorFromLinearColorRGB(Color=$System.TOD Color)
    ScaleA=true
    ScaleRGB=true
   ) {module=ScaleColor002}
   &DynamicMaterialParameters(
    Index 0 Param 1=&Add_Float(
     A=&RandomRangeFloat(Minimum=0, Maximum=1)
     B=&FloatFromCurve(
      FloatCurve=DataInterface() {propertyValues {Curve {keys [
          <at 0 0 RCIM_Cubic>
          {arriveTangent=1.5947, interpMode=RCIM_Cubic, leaveTangent=1.5947, tangentMode=RCTM_User, time=0.2557, value=0.5257}
          <at 1 1 RCIM_Cubic>
         ]}}}
      CurveIndex=$Particles.Age
      Scale Curve=0.25
     )
    )
    Index 0 Param 3=&FloatFromCurve(FloatCurve=DataInterface() {propertyValues {Curve {keys [<at 0 1 RCIM_Cubic>,<at 0.3 0 RCIM_Cubic>]}}})
    Index 0 Param 4=1
    Index 2 Param 1=0.5
    Index 2 Param 3=2
    Module.Index0Float0Write=true
    Module.Index0Float1Write=true
    Module.Index0Float2Write=true
    Module.Index0Float3Write=true
    Module.Index1Float0Write=true
    Module.Index1Float1Write=true
    Module.Index1Float2Write=true
    Module.Index1Float3Write=true
    Module.Index2Float0Write=true
    Module.Index2Float1Write=true
    Module.Index2Float2Write=true
    Module.Index2Float3Write=true
    Module.Index3Float0Write=true
    Module.Index3Float1Write=true
    Module.Index3Float2Write=true
    Module.Index3Float3Write=true
    Write Parameter Index 1=::NewEnumerator1
    Write Parameter Index 2=::NewEnumerator1
    Write Parameter Index 3=::NewEnumerator1
   ) {module=DynamicMaterialParameters002}
   &BlendSpace(Rotation Contribution=0)
  ]
  rendererClasses [&NiagaraSpriteRendererProperties]
  renderers [
   {rendererClass=&NiagaraSpriteRendererProperties}
   {rendererClass=&NiagaraSpriteRendererProperties, rendererIndex=1}
  ]
 }
 NE_Thrusters {
  emitterUpdateScript [
   &EmitterState(Life Cycle Mode=::NewEnumerator1, Loop Behavior=::NewEnumerator1, Loop Duration Mode=::NewEnumerator1, UseLoopDelay=false)
   &SpawnBurst_Instantaneous(Spawn Count=2, Use Spawn Probability=false, UseLoopCountLimit=false)
   /SetVariables(Emitter.ScaleMultiplier=1) {isSetParameters=true}
   /SetVariables(Emitter.SpriteSizeScaler:NiagaraFloat=&Multiply_Float(
     A=&Waveform(
      Waveform Count=::NewEnumerator1
      "Waveform [1]"=::NewEnumerator12
      "[1] Frequency"=12
      "[1] Amplitude Min/Max"=<0.9 1.1>
      "Waveform [2] Blend"=::NewEnumerator2
      "Waveform [2]"=::NewEnumerator12
      "[2] Input Position"=0
      "[2] Phase"=$Emitter.Age
      "[2] Phase Scale"=1
      "[2] Compound Frequency"=2
      "[2] Frequency"=20
      "[2] Amplitude Min/Max"=<0.9 1.1>
      "[2] Scale"=1
      "[2] Bias"=0
      "[2] Exponent"=1
      "[2] Pulse Duration"=0.1
      "[2] Clip Triangle Min/Max"=<0 0>
      "[2] Random Seed"=0
     )
     B=$Emitter.ScaleMultiplier
    )) {isSetParameters=true}
  ]
  particleSpawnScript [
   /SetVariables(Particles.SpawnGroup=$Emitter.SpawnGroup) {isSetParameters=true}
   &InitializeParticle(
    Color=&MergeHelper_ColorSwitch(0=<rgba 2 0.4565 0.0269 0.1>, 1=<rgba 1 0.2492 0 0.2>, 2=?, 3=?, Index=$Particles.SpawnGroup, Num=::NewEnumerator0)
    Lifetime=3
    Mesh Scale=&MergeHelper_VectorSwitch(
     0=<0.2 1 1>
     1=&RandomRangeVector(Minimum=<0.1 0.2 0.3>, Maximum=<0.2 0.5 0.5>)
     2=?
     3=?
     Index=$Particles.SpawnGroup
     Num=::NewEnumerator0
    )
    Mesh Scale Mode=::NewEnumerator4
    Position Offset=<-8 0 0>
    Sprite Size=<50 25>
    Sprite Size Mode=::NewEnumerator4
    UsePositionOffset=true
    Write Material Random=true
   ) {module=InitializeParticle001}
   &InitialMeshOrientation(Mesh Orientation Mode=::NewEnumerator2, Orientation Vector=<-1 0 0>, Use Rotation Vector=false)
  ]
  particleUpdateScript [
   &ParticleState(Kill Particles When Lifetime Has Elapsed=false, Let Infinitely Lived Particles Die When Emitter Deactivates=true)
   &UpdateMeshOrientation(Rotation Rate=8) {module=UpdateMeshOrientation001}
   &DynamicMaterialParameters(
    Index 0 Param 1=&FloatFromCurve(FloatCurve=DataInterface() {propertyValues {Curve {keys [<at 0 1>,<at 1 0>]}}})
    Index 0 Param 2=5
    Index 0 Param 3=20
    Index 0 Param 4=1
    Module.Param0WriteEnabled=true
    Module.Param1WriteEnabled=true
    Module.Param2WriteEnabled=true
    Module.Param3WriteEnabled=true
   )
   &ScaleMeshSize(Scale Factor=&VectorFromFloat(Value=&Multiply_Float(
      A=&Waveform("Waveform [1]"=::NewEnumerator12, "[1] Frequency"=35, "[1] Amplitude Min/Max"=<0.75 1.25>)
      B=$Emitter.ScaleMultiplier
     )))
   &CameraOffset()
   &SolveForcesAndVelocity(Clamp Velocity=false, Limit Acceleration=false)
   &AlignSpriteToMeshOrientation(
    Mesh Orientation Relative Sprite Alignment Vector=<1 0 0>
    Mesh Orientation Relative Sprite Facing Vector:Vector3f=<0 0 1>
    SetAlignment=true
    SetFacing=true
   )
   &ScaleSpriteSize(
    Scale Sprite Size Mode=::NewEnumerator0
    Uniform Scale Factor=&Multiply_Float(
     A=&Waveform(
      Waveform Count=::NewEnumerator1
      "Waveform [1]"=::NewEnumerator12
      "[1] Frequency"=12
      "[1] Amplitude Min/Max"=<0.9 1.1>
      "Waveform [2] Blend"=::NewEnumerator2
      "Waveform [2]"=::NewEnumerator12
      "[2] Input Position"=0
      "[2] Phase"=$Emitter.Age
      "[2] Phase Scale"=1
      "[2] Compound Frequency"=2
      "[2] Frequency"=20
      "[2] Amplitude Min/Max"=<0.9 1.1>
      "[2] Scale"=1
      "[2] Bias"=0
      "[2] Exponent"=1
      "[2] Pulse Duration"=0.1
      "[2] Clip Triangle Min/Max"=<0 0>
      "[2] Random Seed"=0
     )
     B=$Emitter.ScaleMultiplier
    )
   )
   &FN_Sprite_Attributes(
    Color=&Multiply_LinearColorByFloat(
     Linear Color=<rgba 1.5 0.4886 0.0075 1>
     Float=&Waveform("Waveform [1]"=::NewEnumerator12, "[1] Frequency"=12, "[1] Amplitude Min/Max"=<0.8 1.25>, "[1] Scale"=0.35)
    )
    Create Color Attribute=true
    Create SpriteSize Attribute=true
    SpriteSize=<200 200>
    Write to Module Namespace=true
   )
   &FN_Mesh_Attributes(Color=<rgba 1 1 1 0.01>, Create Color Attribute=true, Write to Module Namespace=true)
  ]
  rendererClasses [&NiagaraMeshRendererProperties, &NiagaraSpriteRendererProperties]
  renderers [
   {rendererClass=&NiagaraMeshRendererProperties, rendererIndex=0}
   {rendererClass=&NiagaraMeshRendererProperties, rendererIndex=1}
   {rendererClass=&NiagaraSpriteRendererProperties, rendererIndex=2}
   {rendererClass=&NiagaraMeshRendererProperties, rendererIndex=3}
   {rendererClass=&NiagaraMeshRendererProperties, rendererIndex=4}
  ]
 }
}
summary {
 systemName=NS_Emote_GoudaWheel_SmokeTrailPlane
 emitters [
  {emitterName=NE_Smoke, rendererClasses [&NiagaraSpriteRendererProperties]}
  {emitterName=NE_Thrusters, rendererClasses [&NiagaraMeshRendererProperties, &NiagaraSpriteRendererProperties]}
  {emitterName=Light, rendererClasses []}
  {emitterName=Glow, rendererClasses []}
 ]
 userVariables [
  User.alpha {type=NiagaraFloat, defaultValue=0, description=""}
  User.bIsFrontEnd {type=NiagaraBool}
  User.bIsFrontEndPreview {type=NiagaraBool}
  User.OwnerLocalVelocity {type=Vector3f, description="", defaultValue {struct=Vector3f, value=<0 0 0>}}
 ]
}
