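# Niagara system for the GoudaWheel traversal emote, adding dust trails and bursts of tumbling mesh debris.
# Default_Emote_TraversalDust spawns sprites at two or four sockets, gated by owner speed in gameplay, with speed-scaled drag and time-of-day tinting.
# Bullet_Drop launches randomly colored meshes, toggling two collision setups by platform quality and front-end preview.
# User.VehicleIndex switches spawning and socket layouts; the other five emitters list no renderers.
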NiagaraSystem /BRCosmetics/Animation/Game/MainPlayer/Emotes/GoudaWheel/FX/NS_Emote_GoudaWheel_SmokeTrailTrain
dependencies {
 usedDataInterfaces [
  &NiagaraDataInterfaceArrayInt32
  &NiagaraDataInterfaceArrayFloat3
  &NiagaraDataInterfaceCurve
  &NiagaraDataInterfaceMeshRendererInfo
  &NiagaraDataInterfaceArrayColor
  &NiagaraDataInterfacePlatformSet
  &NiagaraDataInterfaceSocketReader
 ]
 usedDynamicInputs [
  &Vector2DFromFloat
  &MakeLinearColorFromVectorAndFloat
  &MultiplyLinearColor
  &MakeVector4_FromLinearColor
  &FNTimeOfDayScalar
 ]
 usedModules [
  &self:SystemSpawnScript.SystemScriptSource.SystemScriptGraph.NiagaraNodeAssignment_3.SetVariables_#1
  &self:SystemSpawnScript.SystemScriptSource.SystemScriptGraph.NiagaraNodeAssignment_4.SetVariables_#2
  &self:SystemSpawnScript.SystemScriptSource.SystemScriptGraph.NiagaraNodeAssignment_2.SetVariables_#3
  &SystemState
  &self:SystemSpawnScript.SystemScriptSource.SystemScriptGraph.NiagaraNodeAssignment_0.SetVariables_#4
  &BlendSpaceHelper
 ]
}
emitters {
 Bullet_Drop {
  emitterUpdateScript [
   &EmitterState(Life Cycle Mode=::NewEnumerator1, Loop Behavior=::NewEnumerator1, Loop Duration=2, UseLoopDelay=false)
   &SpawnBurst_Instantaneous(
    Spawn Count=3
    Spawn Probability=&MakeCustomFloatFromBool(
     Bool=&SetBoolByFloatComparison(A=$User.VehicleIndex, B=1, Comparison Type=::NewEnumerator4)
     True Float=0
     False Float=1
    )
    Spawn Time=0.225
    Use Spawn Probability=true
    UseLoopCountLimit=false
   ) {module=SpawnBurst_Instantaneous001}
   &SpawnBurst_Instantaneous(
    Spawn Count=3
    Spawn Probability=&MakeCustomFloatFromBool(
     Bool=&SetBoolByFloatComparison(A=$User.VehicleIndex, B=1, Comparison Type=::NewEnumerator4)
     True Float=0
     False Float=1
    )
    Spawn Time=0.125
    Use Spawn Probability=true
    UseLoopCountLimit=false
   )
  ]
  particleSpawnScript [
   &InitializeParticle(
    Color=&FN_Select_Color_from_Time_Of_Day(Morning=<rgba 1 1 1 1>, Day=<rgba 1 1 1 1>, Evening=<rgba 1.5 1.5 1.5 1>, Night=<rgba 3 3 3 1>)
    Lifetime=&RandomRangeFloat(Minimum=1, Maximum=2)
    Mass=50
    Mass Mode=::NewEnumerator1
    Mesh Renderer Array Visibility Mode=::NewEnumerator3
    Mesh Renderer Info=DataInterface() {propertyValues {MeshRenderer=&self:Bullet_Drop_1.NiagaraMeshRendererProperties_1}}
    Mesh Scale Mode=::NewEnumerator3
    Sprite Rotation Mode=::NewEnumerator2
    Sprite Size Mode=::NewEnumerator3
    Uniform Sprite Size=0.75
    UsePositionOffset=false
    Write Material Random=true
   )
   &ShapeLocation(
    MaskSpawnGroup=false
    Non Uniform Scale=<2 2 1.25>
    Offset=<0 0 20>
    Offset Coordinate Space=::Local
    Offset Mode=::NewEnumerator0
    Override Seed=false
    Rotation Angle Type:ENiagara_AngleInput=::NewEnumerator0
    Rotation Mode=::NewEnumerator2
    Sphere Radius=10
    UseOwnerScale=true
    Yaw / Pitch / Roll=<0 0 160>
   )
   &InitialMeshOrientation(Use Rotation Vector=false)
   &AddVelocity(
    Module.Use Falloff Distance=false
    Origin Offset=<25 -10 90>
    Override Seed=false
    Velocity Mode=::NewEnumerator1
    Velocity Speed=&RandomRangeFloat(Minimum=45, Maximum=65)
   ) {module=AddVelocity001}
   &AddVelocity(Velocity=&MakeVector(
     X=&RandomRangeFloat(Minimum=-250, Maximum=250)
     Y=&RandomRangeFloat(Minimum=-250, Maximum=250)
     Z=&RandomRangeFloat(Minimum=450, Maximum=500)
    ))
   &Color(Color=&SelectLinearColorFromArray(
     Array Sampling Mode=::NewEnumerator0
     Direct Array Index=?
     Color Selection Array=DataInterface() {propertyValues {ColorData [<rgba 1 0.0074 0 1>, <rgba 0 0.0672 1 1>, <rgba 0.0161 0.1302 0 1>]}}
     Random Seed=0
    ))
  ]
  particleUpdateScript [
   &GravityForce(Gravity=&RandomRangeVector(Minimum=<-50 -50 -980>, Maximum=<50 50 -980>))
   &Collision(
    Age Colliding Particles=false
    Collision Enabled=&Boolean_AndOperation(
     Bool1=&IsPlatformSetActive(New Platform Set=DataInterface() {propertyValues {Platforms {qualityLevelMask=-8}}})
     Bool2=&Boolean_NotOperation(Bool=$User.bIsFrontEndPreview)
    )
    Control Roll On Collision=true
    Enable Rest State=false
    Friction=0.1
    KillOnCollision=false
    Module.EnableMaxCollisionCount=false
    Module.Randomize Collision Normal Vector=false
    Particle Radius Scale=5
    Radius Calculation Type=::NewEnumerator1
    Restitution=0.4
    ScaleRateOfRotation=true
   ) {module=Collision001}
   &Collision(
    Age Colliding Particles=false
    Analytical Collision Plane Space:ENiagaraCoordinateSpace=::Local
    CPU Collision Type=::NewEnumerator3
    Collision Enabled=&Boolean_OrOperation(
     Bool1=&IsPlatformSetActive(New Platform Set=DataInterface() {propertyValues {Platforms {qualityLevelMask=-28}}})
     Bool2=$User.bIsFrontEndPreview
    )
    Control Roll On Collision=true
    Enable Rest State=false
    Friction=0.1
    KillOnCollision=false
    Module.EnableMaxCollisionCount=false
    Module.Randomize Collision Normal Vector=false
    Particle Radius Scale=5
    Perform 2 Analytical Collision Queries=false
    Radius Calculation Type=::NewEnumerator1
    Restitution=0.8
    ScaleRateOfRotation=false
   ) {module=Collision002}
   &MeshRotationForce(Rotation=&VectorFromFloat(Value=&RandomRangeFloat(Minimum=150, Maximum=250)))
   &Drag(
    Drag=&RandomRangeFloat(Minimum=0.8, Maximum=1.2)
    Rotational Drag=0.2
    Use Linear Drag=true
    Use Rotational Drag=true
   )
   &SolveForcesAndVelocity(Clamp Velocity=false, Limit Acceleration=false)
   &ParticleState()
   &ScaleMeshSize(Scale Factor=&Multiply_VectorByFloat(
     Vector=<1 1 1>
     Float=&FloatFromCurve(FloatCurve=DataInterface() {propertyValues {Curve {keys [<at 0 0>,<at 0.06 1>,<at 0.95 1>,<at 1 0>]}}})
    ))
  ]
  rendererClasses [&NiagaraMeshRendererProperties]
  renderers [{rendererClass=&NiagaraMeshRendererProperties, rendererIndex=0}]
 }
 Default_Emote_TraversalDust {
  emitterSpawnScript [/SetVariables(
    Emitter.FE Particle Spawn Rate:NiagaraFloat=&MakeCustomFloatFromBool(
     Bool=&SetBoolByFloatComparison(A=$User.VehicleIndex, B=0, Comparison Type=::NewEnumerator4)
     True Float=130
     False Float=80
    )
    Emitter.Max Traversal Speed=300
    Emitter.Particle Spawn Rate:NiagaraFloat=&MakeCustomFloatFromBool(
     Bool=&SetBoolByFloatComparison(A=$User.VehicleIndex, B=0, Comparison Type=::NewEnumerator4)
     True Float=130
     False Float=80
    )
   ) {isSetParameters=true}]
  emitterUpdateScript [
   &EmitterState(Life Cycle Mode=::NewEnumerator1, Loop Behavior=::NewEnumerator1, Loop Duration Mode=::NewEnumerator1, UseLoopDelay=false)
   &SpawnRate(
    SpawnRate=&MakeCustomFloatFromBool(
     Bool=&SetBoolByFloatComparison(A=$User.VehicleIndex, B=1, Comparison Type=::NewEnumerator4)
     True Float=0
     False Float=&FN_FrontEndFloatModifier(
      Gameplay=&MakeCustomFloatFromBool(
       Bool=&SetBoolByFloatComparison(A=&VectorLength(Vector=$User.OwnerLocalVelocity), B=150)
       True Float=$Emitter.Particle Spawn Rate
       False Float=0
      )
      Front End=$Emitter.FE Particle Spawn Rate
     )
    )
    Use Spawn Probability=false
   )
  ]
  particleSpawnScript [
   &PartitionParticles(Number Of Partitions=&MakeCustomIntFromBool(
     Bool=&SetBoolByFloatComparison(A=$User.VehicleIndex, B=0, Comparison Type=::NewEnumerator4)
     True INT=4
     False INT=2
    ))
   &InitializeParticle(
    Color Channel Mode:ENiagaraChannelCorrelation=::NewEnumerator0
    Color Maximum=<rgba 0.9068 0.8864 0.8192 0.4>
    Color Minimum=<rgba 0.7818 0.7404 0.6737 0.2>
    Color Mode=::NewEnumerator2
    Lifetime Max=0.5
    Lifetime Min=0.3
    Lifetime Mode=::NewEnumerator1
    Sprite Rotation Mode=::NewEnumerator2
    Sprite Size Mode=::NewEnumerator1
    Uniform Sprite Size Max=30
    Uniform Sprite Size Min=35
    UsePositionOffset=false
    Write Material Random=true
   ) {module=InitializeParticle001}
   &SocketLocation(
    SamplePosOffset=true
    Sampled Position Offset=&MakeCustomVectorFromBool(
     Bool=&SetBoolByFloatComparison(A=$User.VehicleIndex, B=0, Comparison Type=::NewEnumerator4)
     True Vector=&SelectVectorFromArray(
      Direct Array Index=$Particles.PartitionParticles.Partition
      Vector Selection Array=DataInterface() {propertyValues {FloatData [<0 -10 -9>, <0 -10 -9>, <10 -10 -9>, <-10 -10 -9>]}}
     )
     False Vector=&SelectVectorFromArray(
      Direct Array Index=$Particles.PartitionParticles.Partition
      Vector Selection Array=DataInterface() {propertyValues {FloatData [<7 -10 -9>, <-7 -10 -9>]}}
     )
    )
    Sampled Position Offset Coordinate Space:ENiagaraSkelMeshTransforms=::NewEnumerator2
    Socket Index=&MakeCustomIntFromBool(
     Bool=&SetBoolByFloatComparison(A=$User.VehicleIndex, B=0, Comparison Type=::NewEnumerator4)
     True INT=&SelectIntFromArray(
      Direct Array Index=$Particles.PartitionParticles.Partition
      Direct Array Mode=::NewEnumerator1
      Int Selection Array=DataInterface() {propertyValues {IntData [0,1,2,3]}}
     )
     False INT=&SelectIntFromArray(
      Direct Array Index=$Particles.PartitionParticles.Partition
      Direct Array Mode=::NewEnumerator1
      Int Selection Array=DataInterface() {propertyValues {IntData [0,1]}}
     )
    )
    Socket Reader=$System.Socket Reader
    Socket Sampling Mode=::NewEnumerator3
    VelocitySpeedLimit=false
   )
   &AddVelocity(Velocity=<0 2 0.2>, Velocity Speed Scale=&RandomRangeFloat(Minimum=-250, Maximum=-350)) {module=AddVelocity001}
   &CameraOffset(Camera Offset Amount=15)
  ]
  particleUpdateScript [
   &ParticleState()
   &Drag(
    Drag=&FloatFromCurve(
     FloatCurve=DataInterface() {propertyValues {Curve {
        defaultValue=1
        keys [{interpMode=RCIM_Cubic, tangentMode=RCTM_User, time=0, value=0.75}, {interpMode=RCIM_Cubic, tangentMode=RCTM_User, time=0.65, value=1}]
       }}}
     Scale Curve=&Lerp_Float(
      A=20
      B=10
      Alpha=&RemapRange(
       Input Value To Remap=&VectorLength(Vector=$User.OwnerLocalVelocity)
       Input Min=0
       Input Max=$Emitter.Max Traversal Speed
       Output Min=0
       Output Max=1
      )
     )
    )
    Use Linear Drag=true
    Use Rotational Drag=true
   )
   &ScaleSpriteSize(Uniform Curve Sprite Scale=DataInterface() {propertyValues {Curve {keys [
        {arriveTangent=0.8753, interpMode=RCIM_Cubic, leaveTangent=0.8753, tangentMode=RCTM_User, time=0, value=1}
        {arriveTangent=0.0238, interpMode=RCIM_Cubic, leaveTangent=0.0238, tangentMode=RCTM_User, time=1, value=1.3}
       ]}}})
   &DynamicMaterialParameters(
    Index 0 Param 1=&Add_Float(
     A=&RandomRangeFloat(Minimum=0, Maximum=1)
     B=&FloatFromCurve(
      FloatCurve=DataInterface() {propertyValues {Curve {keys [
          <at 0 0 RCIM_Cubic>
          {arriveTangent=1.5947, interpMode=RCIM_Cubic, leaveTangent=1.5947, tangentMode=RCTM_User, time=0.2557, value=0.5257}
          <at 1 1 RCIM_Cubic>
         ]}}}
      CurveIndex=$Particles.Age
      Scale Curve=0.25
     )
    )
    Module.Index0Float0Write=true
    Module.Index0Float1Write=true
    Module.Index0Float2Write=true
    Module.Index0Float3Write=true
   ) {module=DynamicMaterialParameters001}
   &ScaleColor(
    Scale Alpha=&FloatFromCurve(
     FloatCurve=DataInterface() {propertyValues {Curve {keys [
         {arriveTangent=15.7039, interpMode=RCIM_Cubic, leaveTangent=15.7039, tangentMode=RCTM_User, time=0, value=0}
         {arriveTangentWeight=0.0754, interpMode=RCIM_Cubic, leaveTangentWeight=0.2297, tangentMode=RCTM_User, time=0.1684, value=1}
         {arriveTangent=-0.5014, interpMode=RCIM_Cubic, leaveTangent=-0.5014, tangentMode=RCTM_User, time=0.4045, value=1}
         {arriveTangent=-2.9749, interpMode=RCIM_Cubic, leaveTangent=-2.9749, tangentMode=RCTM_User, time=0.9, value=0}
        ]}}}
     Scale Curve=&FN_FrontEndFloatModifier(Gameplay=0.4, Front End=0.4)
    )
    Scale RGB=&MakeVectorFromLinearColorRGB(Color=&FNTimeOfDayColor(
      Night Color=&RandomRangeLinearColor(Minimum=<rgba 0.4138 0.5201 0.6661 1>, Maximum=<rgba 0.781 0.781 0.781 1>)
      Morning Color=&RandomRangeLinearColor(Minimum=<rgba 0.6882 0.6227 0.9531 1>, Maximum=<rgba 0.6063 0.7457 0.9346 1>)
      Day Color=&RandomRangeLinearColor(Minimum=<rgba 0.478 0.4265 0.4121 0.8>, Maximum=<rgba 0.781 0.718 0.6943 0.9>)
      Evening Color=&RandomRangeLinearColor(Minimum=<rgba 0.478 0.478 0.478 1>, Maximum=<rgba 0.781 0.781 0.781 1>)
     ))
    ScaleA=true
    ScaleRGB=true
   )
   &BlendSpace(Rotation Contribution=0)
   &SolveForcesAndVelocity(Clamp Velocity=false, Limit Acceleration=false)
  ]
  rendererClasses [&NiagaraSpriteRendererProperties]
  renderers [
   {rendererClass=&NiagaraSpriteRendererProperties}
   {rendererClass=&NiagaraSpriteRendererProperties, rendererIndex=1}
  ]
 }
 GroundSmokeRing {rendererClasses [], renderers []}
 GroundSmokeRing001 {rendererClasses [], renderers []}
 Smoke001 {rendererClasses [], renderers []}
 Sparkles {rendererClasses [], renderers []}
 TraversalSmoke_Plane {rendererClasses [], renderers []}
}
summary {
 systemName=NS_Emote_GoudaWheel_SmokeTrailTrain
 emitters [
  {emitterName=Default_Emote_TraversalDust, rendererClasses [&NiagaraSpriteRendererProperties]}
  {emitterName=GroundSmokeRing, rendererClasses []}
  {emitterName=TraversalSmoke_Plane, rendererClasses []}
  {emitterName=Bullet_Drop, rendererClasses [&NiagaraMeshRendererProperties]}
  {emitterName=GroundSmokeRing001, rendererClasses []}
  {emitterName=Smoke001, rendererClasses []}
  {emitterName=Sparkles, rendererClasses []}
 ]
 userVariables [
  User.bIsFrontEnd {type=NiagaraBool}
  User.bIsFrontEndPreview {type=NiagaraBool}
  User.OwnerLocalVelocity {type=Vector3f, description="", defaultValue {struct=Vector3f, value=<0 0 0>}}
  User.VehicleIndex {type=NiagaraFloat, defaultValue=0, description=""}
 ]
}
